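# Translucent, surface per-pixel lit material for the bubbles of the RL97 wheel cosmetic, usable on skeletal meshes and mesh deformers.
# It brightens the base color texture and passes the normal map through, with parameters for metallic and roughness.
# Opacity comes from the normal map's blue channel, squared, clamped, scaled down and masked by base color alpha, so areas where the normal tilts become more transparent.
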Material /VehicleCosmetics/Wheels/RL97/Materials/M_Wheel_RL97_Bubbles
graph {
 properties {
  bAutomaticallySetUsageInEditor=false
  bUseMaterialAttributes=true
  bUsedWithMeshDeformer=true
  bUsedWithSkeletalMesh=true
  blendMode=BLEND_Translucent
  translucencyLightingMode=TLM_SurfacePerPixelLighting
  usedShadingModels=Default Lit
 }
 nodes [
  n1 = TextureSampleParameter2D() {group=00 Base, parameterName=Base Color, sortPriority=0, texture=/VehicleCosmetics/Wheels/RL97/Textures/T_RL97_Mushroom_Rim_BC}
  n2 = TextureSampleParameter2D() {
   group=00 Base
   parameterName=Normal
   samplerType=SAMPLERTYPE_Normal
   sortPriority=1
   texture=/VehicleCosmetics/Wheels/RL97/Textures/T_RL97_Mushroom_Rim_N
  }
  n3 = ConstantBiasScale(None=@n2.B) {bias=0}
  n4 = Clamp(None=Power(Base=Add(A=@n3, B=@n3)))
  SetMaterialAttributes!(
   MaterialAttributes
   Base Color=Multiply(A=@n1.RGB) {constB=2}
   Normal=@n2.RGB
   Opacity=Multiply(A=Multiply(A=@n4) {constB=0.4}, B=@n1.A) {constB=0.4}
   Metallic=ScalarParameter() {group=00 Base, parameterName=Metallic, sortPriority=2}
   Roughness=ScalarParameter() {defaultValue=0.5, group=00 Base, parameterName=Roughness, sortPriority=3}
  ) {attributeSetTypes [
    #1
    #2
    #3
    #4
    #5
   ]}
 ]
}
