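# Surface material for the RL97 Mushroom vehicle wheel, giving the rim a tintable base color and an internal glow with faked depth.
# Chained BumpOffset samples with decreasing weights stack into parallax layers: red-channel layers plus a blue channel drive tinted emissive, while green-channel layers brighten base color.
# A static switch picks the base color texture or three custom colors blended by mask channels.
# Material attributes combine these with a normal map and scalar metallic and roughness.
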Material /VehicleCosmetics/Wheels/RL97/Materials/M_Wheel_RL97_Mushroom
graph {
 properties {
  bAutomaticallySetUsageInEditor=false
  bUseMaterialAttributes=true
  bUsedWithMeshDeformer=true
  bUsedWithSkeletalMesh=true
  previewMesh=/VehicleCosmetics/Wheels/RL97/Meshes/SK_Wheel_RL97
  usedShadingModels=Default Lit
 }
 nodes [
  NamedRerouteDeclaration(None=TextureCoordinate()) {name=Coords, nodeColor=<rgba 1 0.8203 0 1>}
  NamedRerouteDeclaration(None=ScalarParameter() {defaultValue=-0.75, group=00 Base, parameterName=LayerDepth, sortPriority=7}) {name=LayerDepth, nodeColor=<rgba 0.5 0 1 1>}
  n1 = NamedRerouteUsage()
  n2 = NamedRerouteUsage()
  n3 = BumpOffset(Coordinate=NamedRerouteUsage(), Height=@n2)
  n4 = Add(
   A=TextureSample(Tex=@n1) {samplerSource=SSM_Wrap_WorldGroupSettings, samplerType=SAMPLERTYPE_LinearColor}.R
   B=TextureSample(UVs=@n3, Tex=@n1) {samplerSource=SSM_Wrap_WorldGroupSettings, samplerType=SAMPLERTYPE_LinearColor}.R
  )
  n5 = BumpOffset(Coordinate=@n3, Height=@n2)
  n6 = Multiply(A=TextureSample(UVs=@n5, Tex=@n1) {samplerSource=SSM_Wrap_WorldGroupSettings, samplerType=SAMPLERTYPE_LinearColor}.R) {constB=0.6}
  n7 = BumpOffset(Coordinate=@n5, Height=@n2)
  n8 = Multiply(A=TextureSample(UVs=@n7, Tex=@n1) {samplerSource=SSM_Wrap_WorldGroupSettings, samplerType=SAMPLERTYPE_LinearColor}.R) {constB=0.3}
  n9 = Multiply(A=TextureSample(UVs=BumpOffset(Coordinate=@n7, Height=@n2), Tex=@n1) {samplerSource=SSM_Wrap_WorldGroupSettings, samplerType=SAMPLERTYPE_LinearColor}.R) {constB=0.05}
  n10 = TextureSample(Tex=NamedRerouteUsage()) {
   samplerSource=SSM_Wrap_WorldGroupSettings
   samplerType=SAMPLERTYPE_LinearColor
   texture=/VehicleCosmetics/Wheels/RL97/Textures/T_RL97_Rim_RGB
  }
  n11 = Add(A=Add(A=Add(A=@n4, B=@n6), B=Add(A=@n8, B=@n9)), B=@n10.B)
  n12 = VectorParameter() {group=00 Base, parameterName=Internal Glow, sortPriority=0, defaultValue=<rgba 0.3278 0.2195 0.5029 1>}
  NamedRerouteDeclaration(None=Multiply(A=Multiply(A=@n11, B=@n12.RGB), B=ScalarParameter() {defaultValue=0.75, group=00 Base, parameterName=Internal Glow Strength, sortPriority=8})) {name=Emissive, nodeColor=<rgba 0 1 0.25 1>}
  n13 = TextureObjectParameter() {
   group=00 Base
   parameterName=Masks
   samplerType=SAMPLERTYPE_LinearColor
   sortPriority=3
   texture=/VehicleCosmetics/Wheels/RL97/Textures/T_RL97_Rim_M
  }
  NamedRerouteDeclaration(None=@n13) {name=Masks, nodeColor=<rgba 0.0312 1 0 1>}
  n14 = NamedRerouteUsage()
  n15 = BumpOffset(Coordinate=NamedRerouteUsage(), Height=@n14)
  n16 = NamedRerouteUsage()
  n17 = BumpOffset(Coordinate=@n15, Height=@n14)
  n18 = TextureSample(UVs=@n17, Tex=@n16) {samplerSource=SSM_Wrap_WorldGroupSettings, samplerType=SAMPLERTYPE_LinearColor}
  n19 = BumpOffset(Coordinate=@n17, Height=@n14)
  n20 = TextureSample(UVs=@n19, Tex=@n16) {samplerSource=SSM_Wrap_WorldGroupSettings, samplerType=SAMPLERTYPE_LinearColor}
  n21 = TextureSample(UVs=BumpOffset(Coordinate=@n19, Height=@n14), Tex=@n16) {samplerSource=SSM_Wrap_WorldGroupSettings, samplerType=SAMPLERTYPE_LinearColor}
  n22 = Add(
   A=Add(
    A=TextureSample(UVs=@n15, Tex=@n16) {samplerSource=SSM_Wrap_WorldGroupSettings, samplerType=SAMPLERTYPE_LinearColor}.G
    B=Multiply(A=@n18.G) {constB=0.6}
   )
   B=Add(A=Multiply(A=@n20.G) {constB=0.3}, B=Multiply(A=@n21.G) {constB=0.05})
  )
  n23 = TextureSampleParameter2D() {
   group=00 Base
   parameterName=Base Color
   samplerSource=SSM_Wrap_WorldGroupSettings
   sortPriority=1
   texture=/VehicleCosmetics/Wheels/RL97/Textures/T_RL97_Rim_BC
  }
  n24 = Multiply(
   A=VectorParameter() {group=00 Base, parameterName=Tint Color, sortPriority=0, defaultValue=<rgba 1 1 1 1>}.RGB
   B=StaticSwitchParameter(True=NamedRerouteUsage(), False=@n23.RGB) {parameterName=Masks_Color}
  )
  n25 = Clamp(None=Multiply(A=Clamp(None=@n22) {maxDefault=0.1}, B=@n24)) {maxDefault=0.2}
  NamedRerouteDeclaration(None=Add(A=Multiply(A=@n22, B=@n25), B=@n24)) {name=BC, nodeColor=<rgba 0 0.3906 1 1>}
  n26 = VectorParameter() {group=00 Base, parameterName=CustomColor2, sortPriority=0, defaultValue=<rgba 0.1221 0.1651 0.3372 1>}
  n27 = VectorParameter() {group=00 Base, parameterName=CustomColor1, sortPriority=0, defaultValue=<rgba 0.3278 0.2195 0.5029 1>}
  n28 = TextureSample(Tex=NamedRerouteUsage()) {samplerSource=SSM_Wrap_WorldGroupSettings, samplerType=SAMPLERTYPE_LinearColor}
  n29 = LinearInterpolate(A=LinearInterpolate(A=@n26.RGB, B=@n27.RGB, Alpha=@n28.R), B=@n26.RGB, Alpha=@n28.G)
  n30 = VectorParameter() {group=00 Base, parameterName=CustomColor3, sortPriority=0, defaultValue=<rgba 0.9047 1 1 1>}
  NamedRerouteDeclaration(None=LinearInterpolate(A=LinearInterpolate(A=@n29, B=@n30.RGB, Alpha=@n28.B), B=@n27.RGB, Alpha=@n28.A)) {name=Masks_Color, nodeColor=<rgba 0 0.9297 1 1>}
  n31 = TextureSampleParameter2D() {
   group=00 Base
   parameterName=Normal
   samplerType=SAMPLERTYPE_Normal
   sortPriority=2
   texture=/VehicleCosmetics/Wheels/RL97/Textures/T_RL97_Mushroom_Rim_N
  }
  SetMaterialAttributes!(
   MaterialAttributes
   Base Color=NamedRerouteUsage()
   Emissive Color=NamedRerouteUsage()
   Normal=@n31.RGB
   Metallic=ScalarParameter() {group=00 Base, parameterName=Metallic, sortPriority=5}
   Roughness=ScalarParameter() {defaultValue=0.5, group=00 Base, parameterName=Roughness, sortPriority=6}
  ) {attributeSetTypes [
    #1
    #2
    #3
    #4
    #5
   ]}
 ]
}
